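# Skeletal-mesh-compatible Default Lit material for the paper prop in the player's Trademark emote, revealing or hiding a signature on the page.
# A signature mask builds a near-black ink tint that multiplies the paper diffuse texture, leaving unmasked areas unchanged.
# The scalar Reveal parameter drives MF_Erosion with a narrow softness, whose Increased Range output blends between signed and plain paper.
# A separate masks texture accompanies the paper diffuse.
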Material /Game/Animation/Game/MainPlayer/Emotes/TradeMark/FX/M_Emote_Trademark_PageReveal
graph {
 properties {bUsedWithMeshDeformer=true, bUsedWithSkeletalMesh=true, pixelDepthOffsetMode=PDOM_Legacy, usedShadingModels=Default Lit}
 nodes [
  Constant()
  n1 = TextureSample() {texture=/Game/Animation/Props/Gadgets_Trademark_Paper/Textures/T_Emote_Prop_Trademark_Paper_Emote_Prop_Trademark_Paper_D}
  n2 = TextureSample() {texture=/Game/Animation/Props/Gadgets_Trademark_Paper/Textures/T_FX_Emote_Trademark_Signature_Mask}
  n3 = Multiply(A=Constant3Vector() {constant=<rgba 0.0104 0.0104 0.0104 1>}, B=@n2.RGB)
  n4 = OneMinus(None=@n2.RGB)
  LinearInterpolate(
   A=Multiply(A=@n1.RGB, B=Add(A=@n3, B=@n4))
   B=@n1.RGB
   Alpha=MaterialFunctionCall!(Gradient, Erosion Amount=ScalarParameter() {parameterName=Reveal}, Softness=Constant() {r=0.01}) {materialFunction=/Game/Effects/Fort_Effects/MaterialFunctions/MF_Erosion}.Increased Range
  ) {constAlpha=1}
  TextureSample() {
   samplerType=SAMPLERTYPE_Masks
   texture=/Game/Animation/Props/Gadgets_Trademark_Paper/Textures/T_Emote_Prop_Trademark_Paper_Emote_Prop_Trademark_Paper_S
  }
 ]
}
